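# Animation Blueprint for the IvoryBrite backpack cosmetic, adjusting pack placement and adding secondary motion to its zipper, bow and wings.
# Chained additive transforms offset and yaw the pack bone, with one offset applied only to non-Large bodies.
# The cached result feeds Anim Dynamics nodes on dyn_zipper, dyn_bow and both dyn_bWing bones in root-relative space.
# On customization load, the event graph stores the body mesh and sets IsLarge from body type.
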AnimBlueprint /BRCosmetics/Accessories/FORT_Backpacks/Backpack_F_MED_IvoryBrite/Meshes/F_MED_IvoryBrite_Pack_AnimBP
target_skeleton=&/F_MED_IvoryBrite_Pack_Skeleton
graphs [
 AnimGraph {nodes [
   Output Pose(Result=Use cached pose 'CRCache_Dyn'()) {layerGroup=None}
   n1 = "Transform (Modify) Bone - Bone: pack"!(ComponentPose=Local To Component(LocalPose=Input Pose()), bAlphaBoolEnabled:bool, Translation="0,-1.35,0") -> (Pose) {alphaInputType=Bool, translationMode=BMM_Additive, translationSpace=BCS_BoneSpace, boneToModify {boneName=pack}, translation {y=-1.35}}
   n2 = "Transform (Modify) Bone - Bone: pack"(ComponentPose=@n1, bAlphaBoolEnabled=NOT Boolean(A=Get IsLarge() -> (IsLarge:bool)), Translation="0,-0.75,0") {alphaInputType=Bool, translationMode=BMM_Additive, translationSpace=BCS_BoneSpace, boneToModify {boneName=pack}, translation {y=-0.75}}
   n3 = "Transform (Modify) Bone - Bone: pack"!(ComponentPose=@n2, Alpha:float, Rotation:Rotator="0,-15,0") -> (Pose) {
    rotationMode=BMM_Additive
    rotationSpace=BCS_BoneSpace
    translationMode=BMM_Additive
    translationSpace=BCS_BoneSpace
    alphaScaleBiasClamp {bInterpResult=true}
    boneToModify {boneName=pack}
    rotation {yaw=-15}
   }
   n4 = "Transform (Modify) Bone - Bone: pack"(ComponentPose=@n3, Rotation="0,15,0") {
    rotationMode=BMM_Additive
    rotationSpace=BCS_BoneSpace
    translationMode=BMM_Additive
    translationSpace=BCS_BoneSpace
    alphaScaleBiasClamp {bInterpResult=true}
    boneToModify {boneName=pack}
    rotation {yaw=15}
   }
   n5 = "Transform (Modify) Bone - Bone: pack"(ComponentPose=@n4, Rotation="0,10,0") {
    rotationMode=BMM_Additive
    rotationSpace=BCS_BoneSpace
    translationSpace=BCS_BoneSpace
    alphaScaleBiasClamp {bInterpResult=true, interpSpeedDecreasing=15, interpSpeedIncreasing=15}
    boneToModify {boneName=pack}
    rotation {yaw=10}
   }
   Save cached pose 'CRCache_Input'(Pose=Component To Local(ComponentPose=@n5)) {cacheName=CRCache_Input}
   n6 = Anim Dynamics - Bone: dyn_zipper!(ComponentPose=Local To Component(LocalPose=Use cached pose 'CRCache_Input'()), Alpha=1, GravityOverride:Vector="0,0,-980") -> (Pose) {
    angularDampingOverride=1
    angularSpringConstant=20000
    bOverrideAngularDamping=true
    bUseGravityOverride=true
    gravityScale=0.35
    simulationSpace=RootRelative
    boundBone {boneName=dyn_zipper}
    componentAppliedLinearAccClamp=<3000 3000 3000>
    componentLinearAccScale=<0.6 0.6 0.6>
    componentLinearVelScale=<0.6 0.6 0.6>
    gravityOverride {z=-980}
    physicsBodyDefinitions [{
      boundBone {boneName=dyn_zipper}
      boxExtents=<10 16 8>
      constraintSetup {angularLimitsMax=<5 20 30>, angularLimitsMin=<-5 -70 -60>, angularTarget=<1 0 0>}
      localJointOffset=<4 0 0>
     }]
   }
   n7 = Anim Dynamics - Bone: dyn_bow!(ComponentPose=@n6, Alpha=0.6, GravityOverride:Vector="0,0,-980") -> (Pose) {
    angularDampingOverride=0.9999
    angularSpringConstant=10000
    bAngularSpring=true
    bOverrideAngularDamping=true
    bUseGravityOverride=true
    gravityScale=0.35
    simulationSpace=RootRelative
    alphaBoolBlend {blendOption=Linear}
    boundBone {boneName=dyn_bow}
    componentAppliedLinearAccClamp=<5000 5000 5000>
    componentLinearAccScale=<1 1 1>
    componentLinearVelScale=<1 1 1>
    gravityOverride {z=-980}
    physicsBodyDefinitions [{
      boundBone {boneName=dyn_bow}
      boxExtents=<30 35 45>
      constraintSetup {angularLimitsMax=<10 5 5>, angularLimitsMin=<-10 -5 -5>, angularTarget=<1 0 0>}
      localJointOffset=<2 0 0>
     }]
    simSpaceSettings {maxAngularAcceleration=30, maxAngularVelocity=6, simSpaceAngularAlpha=1}
   }
   n8 = Anim Dynamics - Bone: dyn_bWing_L!(ComponentPose=@n7, Alpha=0.9, GravityOverride:Vector="0,0,-980") -> (Pose) {
    angularDampingOverride=0.9999
    angularSpringConstant=10000
    bAngularSpring=true
    bOverrideAngularDamping=true
    bUseGravityOverride=true
    gravityScale=0.35
    simulationSpace=RootRelative
    alphaBoolBlend {blendOption=Linear}
    boundBone {boneName=dyn_bWing_L}
    componentAppliedLinearAccClamp=<5000 5000 5000>
    componentLinearAccScale=<1 1 1>
    componentLinearVelScale=<1 1 1>
    gravityOverride {z=-980}
    physicsBodyDefinitions [{
      boundBone {boneName=dyn_bWing_L}
      boxExtents=<30 35 45>
      constraintSetup {angularLimitsMax=<15 20 8>, angularLimitsMin=<-15 -5 -5>, angularTarget=<1 0 0>}
      localJointOffset=<1.5 0 0>
     }]
    simSpaceSettings {maxAngularAcceleration=30, maxAngularVelocity=6, simSpaceAngularAlpha=1}
   }
   n9 = Anim Dynamics - Bone: dyn_bWing_R!(ComponentPose=@n8, Alpha=0.9, GravityOverride:Vector="0,0,-980") -> (Pose) {
    angularDampingOverride=0.9999
    angularSpringConstant=10000
    bAngularSpring=true
    bOverrideAngularDamping=true
    bUseGravityOverride=true
    gravityScale=0.35
    simulationSpace=RootRelative
    alphaBoolBlend {blendOption=Linear}
    boundBone {boneName=dyn_bWing_R}
    componentAppliedLinearAccClamp=<5000 5000 5000>
    componentLinearAccScale=<1 1 1>
    componentLinearVelScale=<1 1 1>
    gravityOverride {z=-980}
    physicsBodyDefinitions [{
      boundBone {boneName=dyn_bWing_R}
      boxExtents=<30 35 45>
      constraintSetup {angularLimitsMax=<15 20 8>, angularLimitsMin=<-15 -5 -5>, angularTarget=<1 0 0>}
      localJointOffset=<1.5 0 0>
     }]
    simSpaceSettings {maxAngularAcceleration=30, maxAngularVelocity=6, simSpaceAngularAlpha=1}
   }
   Save cached pose 'CRCache_Dyn'(Pose=Component To Local(ComponentPose=@n9)) {cacheName=CRCache_Dyn}
  ]}
 EventGraph {nodes [
   n1 = Event OnCharacterCustomizationLoaded()
   > Set Body_Mesh(Body_Mesh=Get Array Item with EFort Custom Part Type(Array=@n1.CustomMeshParts))
   > n2 = FindBodyType!(execute, Object:Object Reference=@n1.CustomBaseMesh) -> (True:Exec, False:Exec, OutBodyType:EFortCustomBodyType Enum=NONE)
   > Set IsLarge(IsLarge:bool="Equal (Enum)"(A:EFortCustomBodyType Enum=@n2, B:EFortCustomBodyType Enum=Large)) -> (Output_Get:bool)
  ]}
]
variables [
 ADSFix_L {type=number}
 IsLarge {type=boolean}
 IsRocketLauncher {type=number}
]
